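# Looping distant flight sound for the Hightower QuinnJet vehicle, in the vehicles sound class with flight attenuation.
# It starts after a fixed delay, then an envelope fades it in and settles it to half volume.
# A PassBy parameter continuously drives pitch and inversely mapped volume.
# Quality level picks randomly among four loops or a two-loop subset with wider pitch variation, each branch passing through a modulator.
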SoundCue /Game/Athena/Sounds/Vehicles/QuinnJet/Veh_Hightower_QJ_Fly_Loop_Cue
root=Delay_0
nodes [
 {id=Delay_0, children [Enveloper_0], properties {delayMax=5, delayMin=5}}
 {
  id=Enveloper_0
  children [ModulatorContinuous_0]
  properties {pitchCurve {editorCurveData {keys [{time=0}]}}, volumeCurve {editorCurveData {keys [<at 0 0>,<at 5 1>,<at 15 0.5>]}}}
 }
 {
  id=ModulatorContinuous_0
  children [QualityLevel_1]
  properties {
   pitchModulationParams {default=0, maxOutput=1.1, minOutput=0.9, parameterName=PassBy}
   volumeModulationParams {maxOutput=0.2, minOutput=1, parameterName=PassBy}
  }
 }
 {id=Modulator_1, children [Random_0]}
 {id=Modulator_2, children [Random_1], properties {pitchMax=1.15, pitchMin=0.85}}
 {id=QualityLevel_1, children [Modulator_1,Modulator_1,Modulator_2,Modulator_2]}
 {id=Random_0, children [WavePlayer_0,WavePlayer_1,WavePlayer_2,WavePlayer_3], properties {weights [1,1,1,1]}}
 {id=Random_1, children [WavePlayer_4,WavePlayer_5], properties {bShouldExcludeFromBranchCulling=true, weights [1,1]}}
 {id=WavePlayer_0, properties {bLooping=true, soundWaveAssetPtr=&/Veh_Hightower_QJ_Fly_Dist_Loop_03}}
 {id=WavePlayer_1, properties {bLooping=true, soundWaveAssetPtr=&/Veh_Hightower_QJ_Fly_Dist_Loop_04}}
 {id=WavePlayer_2, properties {bLooping=true, soundWaveAssetPtr=&/Veh_Hightower_QJ_Fly_Dist_Loop_01}}
 {id=WavePlayer_3, properties {bLooping=true, soundWaveAssetPtr=&/Veh_Hightower_QJ_Fly_Dist_Loop_02}}
 {id=WavePlayer_4, properties {bLooping=true, soundWaveAssetPtr=&Veh_Hightower_QJ_Fly_Dist_Loop_03}}
 {id=WavePlayer_5, properties {bLooping=true, soundWaveAssetPtr=&Veh_Hightower_QJ_Fly_Dist_Loop_04}}
]
properties {
 attenuationSettings=&/Veh_Hightower_QJ_Fly_Att
 bPrimeOnLoad=true
 maxDistance=200000
 soundClassObject=&Fort_Vehicles
}
