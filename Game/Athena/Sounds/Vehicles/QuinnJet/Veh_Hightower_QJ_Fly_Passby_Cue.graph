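# Sound cue for the Quinjet vehicle's fly-by pass sound.
# A Random node picks one of four equally weighted fly-by wave variations, excluded from branch culling, and feeds it through a Modulator at the root.
# It plays in the Fort_Vehicles sound class with a dedicated pass-by attenuation asset and a very long maximum audible distance.
# The cue is set to prime on load.
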SoundCue /Game/Athena/Sounds/Vehicles/QuinnJet/Veh_Hightower_QJ_Fly_Passby_Cue
root=Modulator_0
nodes [
 {id=Modulator_0, children [Random_1]}
 {
  id=Random_1
  children [WavePlayer_4,WavePlayer_5,WavePlayer_6,WavePlayer_7]
  properties {bShouldExcludeFromBranchCulling=true, weights [1,1,1,1]}
 }
 {id=WavePlayer_4, properties {soundWaveAssetPtr=&/Veh_Hightower_QJ_FlyBy_01}}
 {id=WavePlayer_5, properties {soundWaveAssetPtr=&/Veh_Hightower_QJ_FlyBy_02}}
 {id=WavePlayer_6, properties {soundWaveAssetPtr=&/Veh_Hightower_QJ_FlyBy_03}}
 {id=WavePlayer_7, properties {soundWaveAssetPtr=&/Veh_Hightower_QJ_FlyBy_04}}
]
properties {
 attenuationSettings=&/Veh_Hightower_QJ_Fly_Passby_Att
 bPrimeOnLoad=true
 maxDistance=200000
 soundClassObject=&Fort_Vehicles
 volumeMultiplier=1
}
