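# Sound cue that plays the landing sound for the QuinnJet vehicle (Hightower).
# A single wave player feeds the Land_01 wave straight to the root with no randomization or modulation.
# It uses the jet's fly attenuation settings and the Fort vehicles sound class, primes on load, and has a very long max audible distance.
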SoundCue /Game/Athena/Sounds/Vehicles/QuinnJet/Veh_Hightower_QJ_Land_Cue
root=WavePlayer_4
nodes [{id=WavePlayer_4, properties {soundWaveAssetPtr=&/Veh_Hightower_QJ_Land_01}}]
properties {
 attenuationSettings=&/Veh_Hightower_QJ_Fly_Att
 bPrimeOnLoad=true
 maxDistance=200000
 soundClassObject=&Fort_Vehicles
}
